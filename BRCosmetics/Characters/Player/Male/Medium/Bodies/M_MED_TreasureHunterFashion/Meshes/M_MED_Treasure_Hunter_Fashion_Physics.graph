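# Physics asset for the male medium Treasure Hunter Fashion body cosmetic, covering its dangling grapple hook, rope and knife plus the legs.
# Constraints chain the hook base through two clips to the middle rope, which branches to the hook and outer rope.
# The knife is constrained to the right thigh with very tight swing limits.
# Thighs and calves provide leg collision, with a more detailed multi-shape body on the left thigh.
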PhysicsAsset /BRCosmetics/Characters/Player/Male/Medium/Bodies/M_MED_TreasureHunterFashion/Meshes/M_MED_Treasure_Hunter_Fashion_Physics
bodies {
 calf_l [{shapeType=Capsule, length=26.0244, radius=7.5414, center=<-16.1842 -2.6397 -0.4217>, rotation=<rot -0.0706 92.0538 -89.3469>}]
 calf_r [{shapeType=Capsule, length=26.0244, radius=7.5414, center=<16.1842 2.6397 0.4217>, rotation=<rot -0.0706 92.0538 90.6531>}]
 dyn_grappleHookBase [{shapeType=Box, extentX=1.01, extentY=1.01, extentZ=1.01, rotation=<rot 0 -90 0>}]
 dyn_grappleHookClip_01 [{shapeType=Box, extentX=3.6107, extentY=2.3358, extentZ=4.5154, rotation=<rot 0 0 90>}]
 dyn_grappleHookClip_02 [{shapeType=Box, extentX=6.189, extentY=3.313, extentZ=3.4996, center=<2.7093 0 0>, rotation=<rot 0 5.4909 89.9999>}]
 dyn_hook [{shapeType=Sphere, radius=5.5, center=<9.2242 0 0>}]
 dyn_knife [{shapeType=Box, extentX=2.9564, extentY=6.5462, extentZ=30.626, center=<-0.1213 -2.3582 -0.3329>, rotation=<rot 0 0.8693 -87.8974>}]
 dyn_ropeMIddle [{shapeType=Box, extentX=12.825, extentY=2.5417, extentZ=16.2181, center=<7.2856 -0.8309 -1.289>, rotation=<rot 0 -98.4286 -91.0562>}]
 dyn_ropeOutside [{shapeType=Box, extentX=12.3391, extentY=1.8991, extentZ=12.7984, center=<3.5 0 0>, rotation=<rot -3.3417 -95.2062 -90.3831>}]
 thigh_l [
  {shapeType=Sphere, radius=11.375, center=<0.0002 -2.4072 3.3575>}
  {shapeType=Sphere, radius=9.75, center=<0 2.0389 5.1385>}
  {shapeType=Capsule, length=20.5994, radius=6.8383, center=<-20.0778 1.8861 -0.4191>, rotation=<rot 85.2704 -0.0003 -180>}
  {shapeType=Box, extentX=19.25, extentY=12, extentZ=3.25, center=<-0.8945 1.354 0.8852>, rotation=<rot 12.8076 -5.7974 -28.8262>}
 ]
 thigh_r [{shapeType=Capsule, length=28.5994, radius=7.0883, center=<20.7736 -0.0046 0.527>, rotation=<rot 87.7781 -167.8385 -179.9999>}]
}
constraints [
 {bone1Name=dyn_ropeMIddle, bone2Name=dyn_grappleHookClip_02, swing1LimitDegrees=20, swing2LimitDegrees=15, twistLimitDegrees=10}
 {
  bone1Name=dyn_grappleHookClip_01
  bone2Name=dyn_grappleHookBase
  swing1LimitDegrees=45
  swing1Motion=Free
  swing2LimitDegrees=45
  swing2Motion=Free
  twistLimitDegrees=45
  twistMotion=Free
 }
 {bone1Name=dyn_grappleHookClip_02, bone2Name=dyn_grappleHookClip_01, swing1LimitDegrees=20, swing2LimitDegrees=10, twistLimitDegrees=10}
 {bone1Name=dyn_hook, bone2Name=dyn_ropeMIddle, swing1LimitDegrees=40, swing2LimitDegrees=40, twistLimitDegrees=25}
 {bone1Name=calf_r, bone2Name=thigh_r, swing1LimitDegrees=45, swing2LimitDegrees=45, twistLimitDegrees=45}
 {
  bone1Name=calf_l
  bone2Name=thigh_l
  swing1LimitDegrees=45
  swing1Motion=Free
  swing2LimitDegrees=45
  swing2Motion=Free
  twistLimitDegrees=45
  twistMotion=Free
 }
 {bone1Name=dyn_ropeOutside, bone2Name=dyn_ropeMIddle, swing1LimitDegrees=20, swing2LimitDegrees=30, twistLimitDegrees=10}
 {bone1Name=dyn_knife, bone2Name=thigh_r, swing1LimitDegrees=1, swing2LimitDegrees=2, twistLimitDegrees=45}
]
