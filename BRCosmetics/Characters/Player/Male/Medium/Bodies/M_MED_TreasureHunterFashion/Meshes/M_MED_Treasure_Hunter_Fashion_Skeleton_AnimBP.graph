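# Post-copy AnimBlueprint for the male medium Treasure Hunter Fashion body, adding crouch adjustment and secondary physics to the parent mesh's pose.
# It copies the attached parent's pose, then additively offsets dyn_grappleHookBase by crouchAlpha, with interpolated blending.
# A RigidBody node then simulates the outfit's physics asset relative to root, with scaled, clamped component motion and overridden gravity.
# The update event sets crouchAlpha from bIsCrouching.
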AnimBlueprint /BRCosmetics/Characters/Player/Male/Medium/Bodies/M_MED_TreasureHunterFashion/Meshes/M_MED_Treasure_Hunter_Fashion_Skeleton_AnimBP
target_skeleton=&/M_MED_Treasure_Hunter_Fashion_Skeleton
graphs [
 AnimGraph {nodes [
   n1 = "Transform (Modify) Bone - Bone: dyn_grappleHookBase"(
    ComponentPose=Local To Component(LocalPose=Copy Pose From Mesh() {bUseAttachedParent=true})
    Alpha=Get crouchAlpha() -> (crouchAlpha:double)
    Translation="-3.440,-9.080,0.392"
    Rotation="4.799,-55.227,-1.484"
    Scale="1,1,1"
   ) {
    rotationMode=BMM_Additive
    rotationSpace=BCS_BoneSpace
    translationMode=BMM_Additive
    translationSpace=BCS_BoneSpace
    alphaBoolBlend {blendOption=Linear}
    alphaScaleBiasClamp {bInterpResult=true, interpSpeedDecreasing=7, interpSpeedIncreasing=7}
    boneToModify {boneName=dyn_grappleHookBase}
    rotation=<rot 4.799 -55.2269 -1.484>
    translation=<-3.4402 -9.0801 0.392>
   }
   n2 = RigidBody(ComponentPose=@n1) {
    bClampLinearTranslationLimitToRefPose=true
    bOverrideWorldGravity=true
    overridePhysicsAsset=&/M_MED_Treasure_Hunter_Fashion_Physics
    simulationSpace=BaseBoneSpace
    alphaBoolBlend {blendOption=Linear}
    baseBoneRef {boneName=root}
    componentAppliedLinearAccClamp=<600 600 2500>
    componentLinearAccScale=<0.3 0.3 0.45>
    componentLinearVelScale=<0.4 0.4 0.375>
    overrideWorldGravity {z=-980}
   }
   Output Pose(Result=Component To Local(ComponentPose=@n2)) {layerGroup=None}
  ]}
 EventGraph {nodes [
   Event BlueprintUpdateAnimation()
   n1 = Get bIsCrouching()
   > Set crouchAlpha(crouchAlpha:double="To Float (Boolean)"(InBool=@n1)) -> (Output_Get:double)
   TryGetPawnOwner()
   NOT Boolean(A=@n1)
  ]}
]
variables [crouchAlpha {type=number}]
