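# Physics asset for the face accessory mesh of the male medium TreasureHunterFashion player character, covering the head and a front hair bone.
# A large capsule body represents the head, and a small capsule body sits on the dyn_hairFront bone.
# One constraint joins the two with narrow swing and even narrower twist limit angles, which points to subtle secondary hair motion relative to the head.
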PhysicsAsset /BRCosmetics/Characters/Player/Male/Medium/Bodies/M_MED_TreasureHunterFashion/Meshes/Parts/M_MED_TreasureHunterFashion_FaceAcc_Physics
bodies {
 dyn_hairFront [{shapeType=Capsule, length=2.287, radius=0.7499, center=<1.2146 -0.0315 0.0008>, rotation=<rot 0 85.4195 -83.6129>}]
 head [{shapeType=Capsule, length=5.7275, radius=7.7443, center=<6.1823 -0.3984 0.2085>, rotation=<rot 0 162.3191 -88.0446>}]
}
constraints [{bone1Name=dyn_hairFront, bone2Name=head, swing1LimitDegrees=5, swing2LimitDegrees=5, twistLimitDegrees=1}]
